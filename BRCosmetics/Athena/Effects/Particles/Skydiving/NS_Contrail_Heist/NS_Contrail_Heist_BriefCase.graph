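# Heist skydiving contrail Niagara system that trails a kite-style tethered prop, a briefcase per the asset name, with speed lines.
# The kite emitter uses shared kite modules with noise curves, diving attributes and spool/string settings, drawing through three mesh renderers and a ribbon.
# Speed-line sprites spawn on a torus, stretch with velocity and shift with owner velocity.
# Front-end modifiers set their spawn rate to zero outside gameplay.
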NiagaraSystem /BRCosmetics/Athena/Effects/Particles/Skydiving/NS_Contrail_Heist/NS_Contrail_Heist_BriefCase
dependencies {
 usedDataInterfaces [
  &NiagaraDataInterfaceCurve
  &NiagaraDataInterfaceVectorCurve
  &NiagaraDataInterfaceVector2DCurve
 ]
 usedDynamicInputs []
 usedModules [&SystemState, /Niagara/Modules/System/SystemLifeCycle]
}
emitters {
 FN_Contrails_Kite_v2001 {
  emitterUpdateScript [
   &EmitterScalability(
    Module.bApplySpawnCountScale=false
    Module.bApplySpawnCountScaleByDistanceFraction=false
    Module.bUseMaxDistance=false
    Module.bUseMinDistance=false
   ) {enabled=false}
   &EmitterLifeCycle(NextLoopDuration=1.9958) {enabled=false}
   &EmitterState()
   /Game/Effects/Niagara/Modules/Spawn/FN_Spawn_Kite(
    "Link Per Kite  Amount"=1
    Noise=DataInterface() {propertyValues {
      XCurve {keys [
        <at 0 0 RCIM_Cubic>
        {arriveTangent=-22.177, interpMode=RCIM_Cubic, leaveTangent=-22.177, time=0.0148, value=0.3241}
        {arriveTangent=8.0832, interpMode=RCIM_Cubic, leaveTangent=8.0832, time=0.0653, value=-1.4471}
        {arriveTangent=39.4964, interpMode=RCIM_Cubic, leaveTangent=39.4964, time=0.1184, value=1.1614}
        <at 0.1293 1.083>
        {arriveTangent=-23.4377, interpMode=RCIM_Cubic, leaveTangent=-23.4377, time=0.1576, value=0.0233}
        {arriveTangent=-0.0218, interpMode=RCIM_Cubic, leaveTangent=-0.0218, time=0.2029, value=-0.642}
        <at 0.267 0.0209>
        {arriveTangent=-9.845, interpMode=RCIM_Cubic, leaveTangent=-9.845, time=0.2888, value=0.1309}
        {arriveTangent=3.938, interpMode=RCIM_Cubic, leaveTangent=3.938, time=0.3474, value=-0.7708}
        {arriveTangent=2.5316, interpMode=RCIM_Cubic, leaveTangent=2.5316, time=0.3869, value=0.5173}
        {arriveTangent=-4.044, interpMode=RCIM_Cubic, leaveTangent=-4.044, time=0.4237, value=-0.5776}
        <at 0.4469 0.2748>
        {arriveTangent=-4.0955, interpMode=RCIM_Cubic, leaveTangent=-4.0955, time=0.4551, value=0.4851}
        {arriveTangent=-8.4961, interpMode=RCIM_Cubic, leaveTangent=-8.4961, time=0.5107, value=0.0134}
        <at 0.5109 0.0104>
        {arriveTangent=-0.6728, interpMode=RCIM_Cubic, leaveTangent=-0.6728, time=0.5736, value=-1.125}
        <at 0.6118 -0.0575>
        {arriveTangent=17.1461, interpMode=RCIM_Cubic, leaveTangent=17.1461, time=0.6146, value=0.0105}
        {arriveTangent=11.3424, interpMode=RCIM_Cubic, leaveTangent=11.3424, time=0.6622, value=0.8072}
        {arriveTangent=-13.2445, interpMode=RCIM_Cubic, leaveTangent=-13.2445, time=0.699, value=0.9682}
        {arriveTangent=-25.0454, interpMode=RCIM_Cubic, leaveTangent=-25.0454, time=0.7226, value=0.0074}
        {arriveTangent=-0.028, interpMode=RCIM_Cubic, leaveTangent=-0.028, time=0.7672, value=-0.7386}
        {arriveTangent=21.9983, interpMode=RCIM_Cubic, leaveTangent=21.9983, time=0.7974, value=0.0053}
        {arriveTangent=5.888, interpMode=RCIM_Cubic, leaveTangent=5.888, time=0.8462, value=1.0004}
        {arriveTangent=-15.155, interpMode=RCIM_Cubic, leaveTangent=-15.155, time=0.8953, value=0.5818}
        <at 0.9119 0.0053>
        {arriveTangent=-7.8259, interpMode=RCIM_Cubic, leaveTangent=-7.8259, time=0.9321, value=-0.6098}
        <at 0.9591 -0.3645>
        {arriveTangent=8.9177, interpMode=RCIM_Cubic, leaveTangent=8.9177, time=0.9879, value=0}
        {arriveTangent=0.682, interpMode=RCIM_Cubic, leaveTangent=0.682, tangentMode=RCTM_User, time=1, value=0}
       ]}
      YCurve {keys [
        <at 0 0 RCIM_Cubic>
        {arriveTangent=0.0179, interpMode=RCIM_Cubic, leaveTangent=0.0179, time=0.0843, value=1.1292}
        {arriveTangent=-19.1001, interpMode=RCIM_Cubic, leaveTangent=-19.1001, time=0.117, value=0.0021}
        <at 0.1293 0.2701>
        {arriveTangent=-34.1213, interpMode=RCIM_Cubic, leaveTangent=-34.1213, time=0.1607, value=1.1225}
        {arriveTangent=-41.5782, interpMode=RCIM_Cubic, leaveTangent=-41.5782, time=0.1702, value=-1.125}
        {arriveTangent=2.3241, interpMode=RCIM_Cubic, leaveTangent=2.3241, time=0.2015, value=-0.5776}
        {arriveTangent=-2.7223, interpMode=RCIM_Cubic, leaveTangent=-2.7223, time=0.2533, value=-0.9318}
        <at 0.267 -0.7557>
        {arriveTangent=19.4781, interpMode=RCIM_Cubic, leaveTangent=19.4781, time=0.3065, value=0.1309}
        {arriveTangent=-8.6722, interpMode=RCIM_Cubic, leaveTangent=-8.6722, time=0.3637, value=1.1292}
        <at 0.4469 -1.0865>
        {arriveTangent=41.2205, interpMode=RCIM_Cubic, leaveTangent=41.2205, time=0.4687, value=-1.3504}
        {arriveTangent=49.9749, interpMode=RCIM_Cubic, leaveTangent=49.9749, time=0.4959, value=0.936}
        <at 0.5109 0.7611>
        {arriveTangent=3.1016, interpMode=RCIM_Cubic, leaveTangent=3.1016, time=0.5532, value=-0.5454}
        {arriveTangent=9.5865, interpMode=RCIM_Cubic, leaveTangent=9.5865, time=0.5777, value=0.9682}
        <at 0.6118 0.0165>
        {arriveTangent=11.842, interpMode=RCIM_Cubic, leaveTangent=11.842, time=0.6309, value=-0.4165}
        {arriveTangent=-2.9535, interpMode=RCIM_Cubic, leaveTangent=-2.9535, time=0.6731, value=0.7428}
        {arriveTangent=-5.1686, interpMode=RCIM_Cubic, leaveTangent=-5.1686, time=0.6854, value=-0.5776}
        {arriveTangent=7.3333, interpMode=RCIM_Cubic, leaveTangent=7.3333, time=0.7167, value=0.5173}
        {arriveTangent=-15.9313, interpMode=RCIM_Cubic, leaveTangent=-15.9313, time=0.7641, value=0}
        <at 0.8095 -0.9606>
        {interpMode=RCIM_Cubic, tangentMode=RCTM_User, time=0.8489, value=0}
        {interpMode=RCIM_Cubic, tangentMode=RCTM_User, time=0.904, value=0.6415}
        {interpMode=RCIM_Cubic, tangentMode=RCTM_User, time=0.9329, value=0.2589}
        {interpMode=RCIM_Cubic, tangentMode=RCTM_User, time=0.967, value=-0.5063}
        {interpMode=RCIM_Cubic, tangentMode=RCTM_User, time=1, value=0}
       ]}
      ZCurve {keys [
        {interpMode=RCIM_Cubic, tangentMode=RCTM_User, time=0, value=0}
        {arriveTangent=10.357, interpMode=RCIM_Cubic, leaveTangent=10.357, tangentMode=RCTM_User, time=0.0489, value=-1.0284}
        {arriveTangent=5.6189, interpMode=RCIM_Cubic, leaveTangent=5.6189, tangentMode=RCTM_User, time=0.0854, value=0.8726}
        {arriveTangent=-0.1676, interpMode=RCIM_Cubic, leaveTangent=-0.1676, tangentMode=RCTM_User, time=0.1245, value=-0.6036}
        {arriveTangent=-0.5858, interpMode=RCIM_Cubic, leaveTangent=-0.5858, tangentMode=RCTM_User, time=0.1956, value=0.8542}
        {interpMode=RCIM_Cubic, tangentMode=RCTM_User, time=0.219, value=-0.659}
        {interpMode=RCIM_Cubic, tangentMode=RCTM_User, time=0.2776, value=0.8542}
        {arriveTangent=-3.0163, interpMode=RCIM_Cubic, leaveTangent=-3.0163, tangentMode=RCTM_User, time=0.2916, value=-0.659}
        {arriveTangent=7.8759, interpMode=RCIM_Cubic, leaveTangent=7.8759, tangentMode=RCTM_User, time=0.3877, value=0.522}
        {arriveTangent=-10.0071, interpMode=RCIM_Cubic, leaveTangent=-10.0071, tangentMode=RCTM_User, time=0.4439, value=0.5405}
        {arriveTangent=-1.5965, interpMode=RCIM_Cubic, leaveTangent=-1.5965, tangentMode=RCTM_User, time=0.5205, value=-0.8066}
        {arriveTangent=1.6215, interpMode=RCIM_Cubic, leaveTangent=1.6215, tangentMode=RCTM_User, time=0.5595, value=0.3559}
        {arriveTangent=2.5933, interpMode=RCIM_Cubic, leaveTangent=2.5933, tangentMode=RCTM_User, time=0.6001, value=-0.6774}
        {arriveTangent=26.7367, interpMode=RCIM_Cubic, leaveTangent=26.7367, tangentMode=RCTM_User, time=0.6235, value=0.522}
        {arriveTangent=-16.0548, interpMode=RCIM_Cubic, leaveTangent=-16.0548, tangentMode=RCTM_User, time=0.6595, value=0.9095}
        {arriveTangent=-7.1505, interpMode=RCIM_Cubic, leaveTangent=-7.1505, tangentMode=RCTM_User, time=0.6845, value=-0.456}
        {arriveTangent=5.7226, interpMode=RCIM_Cubic, leaveTangent=5.7226, tangentMode=RCTM_User, time=0.7188, value=0.4851}
        {interpMode=RCIM_Cubic, tangentMode=RCTM_User, time=0.7641, value=0}
        <at 0.8148 -0.6258>
        {arriveTangent=6.4436, interpMode=RCIM_Cubic, leaveTangent=6.4436, time=0.8699, value=0.7849}
        {interpMode=RCIM_Cubic, tangentMode=RCTM_User, time=0.9119, value=0}
        <at 0.9565 -0.6736 RCIM_Cubic>
        <at 1 0>
       ]}
     }}
   )
  ]
  particleSpawnScript [/Game/Effects/Niagara/Modules/Kite/FN_Initialize_Kite_v2(Module.Write Color=true, Module.Write RibbonTwist=false, Module.Write RibbonWidth=true)]
  particleUpdateScript [
   /Game/Effects/Niagara/Modules/Skydiving/FN_DivingAttributes/FN_DivingAttributes(Accumulate Dive=0.003)
   /Game/Effects/Niagara/Modules/Kite/FN_Update_Kite_v2(
    Amplitude=40
    Axis to match Bone Ref Axis:Vector3f=<0 1 0>
    Bone Reference Axis:Vector3f=<1 0 0>
    Frequency=10
    Kite Scale:Vector3f=&VectorFromFloat(Value=0.8)
    Rotation Amplitude=20
    Scale Kite Position=-31.6522
    Scale String Start Pos=14.4
    Spool Position Offset:Vector3f=<18.9256 -0.1236 2.5576>
    Spool Scale:Vector3f=&VectorFromFloat(Value=0.8)
    String End Offset:Vector3f=<59.6 34.834 97.7>
    String Length Limit=97.79
    String Scale:Vector3f=&VectorFromFloat(Value=0.6)
   )
   /Game/Effects/Niagara/Modules/Materials/FN_MeshMaterial(Amplitude=300, Falloff=416.769, Frequency=200, Scale Noise Y Axis=1.5, Scale Noise Z Axis=0, Speed=1)
   /SetVariables(Particles.FN_KiteMeshAttributes.Orientation:Quat4f=&MultiplyQuaternion(
     Quaternion A=$Particles.FN_KiteMeshAttributes.Orientation
     Quaternion B=&FindQuatBetweenTwoVectors(
      From Vector=&RotateVectorByQuaternion(Vector To Rotate=<0 1 0>, Quaternion=$Particles.FN_KiteMeshAttributes.Orientation)
      To Vector=$Engine.Owner.SystemYAxis
     )
    )) {enabled=false, isSetParameters=true}
  ]
  rendererClasses [&NiagaraMeshRendererProperties, &NiagaraRibbonRendererProperties]
  renderers [
   {rendererClass=&NiagaraMeshRendererProperties, rendererIndex=0}
   {rendererClass=&NiagaraMeshRendererProperties, rendererIndex=1}
   {rendererClass=&NiagaraMeshRendererProperties, rendererIndex=2}
   {rendererClass=&NiagaraRibbonRendererProperties, rendererIndex=3}
  ]
 }
 FN_Sprite_SpeedLines {
  emitterUpdateScript [
   &SpawnBurst_Instantaneous(Spawn Count=5, Use Spawn Probability=false) {enabled=false}
   &EmitterState() {module=EmitterState001}
   &EmitterState()
   &SpawnRate(SpawnRate=&FN_FrontEndFloatModifier(Gameplay=100, Front End=0), Use Spawn Probability=false)
  ]
  particleSpawnScript [
   /Niagara/Modules/Spawn/Initialization/InitializeParticle(
    Color=&MakeLinearColor(
     R:NiagaraFloat=0.2292
     G:NiagaraFloat=0.1989
     B:NiagaraFloat=0.1838
     A:NiagaraFloat=&UniformRangedFloat(Minimum=0.15, Maximum=0.3)
    )
    Lifetime=&UniformRangedFloat(Minimum=0.3, Maximum=0.4)
    Module.Write Color=true
    Module.Write Lifetime=true
    Module.Write Mass=true
    Module.Write Material Random=true
    Module.Write Position=false
    Module.Write Scale=false
    Module.Write SpriteRotation=true
    Module.Write SpriteSize=true
    Position=&SimulationPosition()
    Sprite Rotation=&UniformRangedFloat(Minimum=0, Maximum=360)
    Sprite Size=&UniformRangedVector2D(Minimum=<50 25>, Maximum=<0.8 60>)
   )
   &TorusLocation(
    Align Torus With Axis=true
    MaskSpawnGroup=false
    Module.Override Randomness=false
    Torus Origin:NiagaraPosition=&Add_Vector(
     A=Linked() {linkedVariable {name=Particles.Position, type=NiagaraPosition}}
     B=&MakeVector(X=0, Y=0, Z=&FN_FrontEndFloatModifier(Gameplay=0, Front End=100))
    )
   )
   &AddVelocity(Velocity=&MakeVector(
     X=0
     Y=0
     Z=&Multiply_Float(
      A=&UniformRangedFloat(Minimum=50, Maximum=100)
      B=&FN_FrontEndFloatModifier(Gameplay=1, Front End=50)
     )
    ))
   &InheritVelocity(Module.Limit Speed=true, Velocity Scale=<-0.9 -0.9 -0.9>) {enabled=false}
  ]
  particleUpdateScript [
   &ParticleState()
   &Color(Scale Alpha=&FloatFromCurve(FloatCurve=DataInterface() {propertyValues {Curve {keys [<at 0 0 RCIM_Cubic>,<at 0.5 1 RCIM_Cubic>,<at 1 0 RCIM_Cubic>]}}})) {enabled=false}
   &SolveForcesAndVelocity(Clamp Velocity=false, Limit Acceleration=false)
   &ScaleSpriteSizeBySpeed(
    Max Scale Factor=<1 100>
    Min Scale Factor=<0 0.05>
    Source Velocity:Vector3f=&FN_FrontEndVector3Modifier(Gameplay=$Particles.Velocity, Front End=<0 0 500>)
   )
   &ScaleSpriteSize(Scale Factor=&Vector2DFromCurve(
     Vector2Curve=DataInterface() {propertyValues {
       XCurve {keys [<at -0.0069 0.5 RCIM_Cubic>,<at 0.4918 1 RCIM_Cubic>,<at 1 0.5 RCIM_Cubic>]}
       YCurve {keys [<at 0 0.0245 RCIM_Cubic>, {arriveTangent=-0.0245, interpMode=RCIM_Cubic, leaveTangent=-0.0245, time=0.4975, value=1}, <at 1 0 RCIM_Cubic>]}
      }}
     ScaleCurve=<1 1>
    ))
   &CameraOffset(Camera Offset Amount=-100) {enabled=false}
   /SetVariables(Particles.Position=&Add_Vector(
     A=Linked() {linkedVariable {name=Particles.Position, type=NiagaraPosition}}
     B=&Multiply_Vector(
      A=$Engine.Owner.Velocity
      B=&VectorFromFloat(Value=&FN_FrontEndFloatModifier(Gameplay=0.0005, Front End=0))
     )
    )) {isSetParameters=true}
  ]
  rendererClasses [&NiagaraSpriteRendererProperties]
  renderers [{rendererClass=&NiagaraSpriteRendererProperties}]
 }
}
summary {
 systemName=NS_Contrail_Heist_BriefCase
 emitters [
  {
   emitterName=FN_Contrails_Kite_v2001
   rendererClasses [&NiagaraMeshRendererProperties, &NiagaraRibbonRendererProperties]
  }
  {emitterName=FN_Sprite_SpeedLines, rendererClasses [&NiagaraSpriteRendererProperties]}
 ]
 userVariables [
  User.bIsFrontEnd {type=NiagaraBool}
  User.bIsFrontEndPreview {type=NiagaraBool}
  User.PlayerSpeed {type=NiagaraFloat, defaultValue=0, description=""}
  User.RibbonIDOffset {type=NiagaraInt32}
  User.SkeletalMesh {
   type=NiagaraDataInterfaceSkeletalMesh
   defaultValue {
    struct=NiagaraExt_VariableValue_DataInterface
    value {dataInterface=&self:NiagaraDataInterfaceSkeletalMesh_0}
   }
  }
 ]
}
